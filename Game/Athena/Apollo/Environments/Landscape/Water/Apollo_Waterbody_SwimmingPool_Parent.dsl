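; Parent Blueprint for Apollo swimming pool water bodies, judging by its name and path.
; Its construction script assigns the Post Process Material variable to the water body component as the underwater post-process material.
; BeginPlay is present but empty, so the graph adds no runtime logic.
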(event EventBeginPlay)

(fn ConstructionScript
  (Rendering|SetUnderwaterPostProcessMaterial (Variables|Water|GetWaterBodyComponent) (Variables|Water|GetPostProcess%20Material)))
