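; Parent Blueprint for Apollo waterfall transition water bodies, which spawns effects where waterfalls occur.
; The construction script moves an Arrow component to the WaterBottom position and scales it, stretching X with the waterfall width.
; BeginPlay calls CreateFX, whose body is not in this graph.
; A custom CE_TestNiagara event makes the same call, apparently for testing the Niagara effects.
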(event EventBeginPlay
  (Utilities|CreateFX))

(event Custom|CE_TestNiagara
  (Utilities|CreateFX))

(fn ConstructionScript
  (bind _arrow (Variables|Default|GetArrow))
  (Transformation|SetRelativeLocation _arrow (Variables|Effects|GetWaterBottom) false false)
  (Transformation|SetWorldScale3D _arrow (Math|Vector|MakeVector (* (Variables|Effects|GetWaterFallWidth) 4.0) 4.0 4.0)))
