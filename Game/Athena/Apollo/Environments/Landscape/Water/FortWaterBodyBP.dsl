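; Fortnite water body Blueprint that links its surface to fluid simulation and makes it react to bullet impacts, only where cosmetic events can execute.
; On begin play, it sets material parameters through the Shallow Water subsystem, or caches the first BP_FluidSim_FN actor and passes itself for texture setup.
; Bullet hits register with the subsystem, falling back to a force impulse on that actor and a texture reapply, marked as needed for Juno.
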(event EventBeginPlay
  (bind _returnvalue (WorldSubsystems|GetShallowWaterSubsystem))
  (Networking|CanExecuteCosmeticEvents
    (:True
      (if (Utilities|IsValid _returnvalue :!class /Script/Engine.KismetSystemLibrary)
        (ShallowWater|SetWaterBodyMIDParameters _returnvalue self)
        (else
          (bind _outactors (Actor|GetAllActorsOfClass "/Game/Athena/Apollo/Water/Blueprints/BP_FluidSim_FN.BP_FluidSim_FN_C"))
          (if (> (Utilities|Array|Length _outactors) 0)
            (bind _output_get (Variables|Default|SetFluidSimBP (Utilities|Array|Get(acopy) _outactors)))
            (Class|BPFluidSim01|SetSingleWaterBodyMIDTexture _output_get self) ; Needed for Juno
          ))))))

(event WaterBody|EventOnReceivedBulletImpact
  (params (Hit :type Hit%20Result%20Structure%20%28by%20ref%29))
  (bind (_bblockinghit _binitialoverlap _time _distance _location) (Collision|BreakHitResult Hit))
  (bind _returnvalue (WorldSubsystems|GetShallowWaterSubsystem))
  (bind _fluidsimbp (Variables|Default|GetFluidSimBP))
  (Networking|CanExecuteCosmeticEvents
    (:True
      ;[0.1161 1 0.3371] Temp solution before FortWaterAdvanced is enabled for realzz #jira FORT-700640
      (if (Utilities|IsValid _returnvalue :!class /Script/Engine.KismetSystemLibrary)
        (ShallowWater|RegisterImpact _returnvalue _location "0, 0, 0" 32.0)
        (else
          (Class|BPFluidSim01|ApplyFluidForceImpulse _fluidsimbp (Utilities|Struct|MakeFluidForceImpulse _location 32.0 -1e+01 "/FortWater/FluidSimulation/Materials/Forces/M_Fluid_Sim_Force_Impulse.M_Fluid_Sim_Force_Impulse" 0.0 0.0 (Utilities|Struct|MakeFluidForceImpulseTimedEffects true 7e+01 -2.0 1e+02 0.3 0.25 1.0)))
          (Class|BPFluidSim01|SetSingleWaterBodyMIDTexture _fluidsimbp self) ; Needed for Juno
        )))))

(fn ConstructionScript)
